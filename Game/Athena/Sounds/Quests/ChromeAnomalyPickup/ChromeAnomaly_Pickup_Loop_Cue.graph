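# Looping ambient sound cue for the Chrome Anomaly quest pickup, routed through the quests sound class.
# A random node picks one of three equally weighted looping waves.
# A modulator above it applies slight downward pitch variation and a volume boost range to each playback.
# It references the supply drop Hidden Star ambient loop attenuation asset and also carries falloff overrides on the cue.
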SoundCue /Game/Athena/Sounds/Quests/ChromeAnomalyPickup/ChromeAnomaly_Pickup_Loop_Cue
root=Modulator_0
nodes [
 {id=Modulator_0, children [Random_0], properties {pitchMax=1, pitchMin=0.97, volumeMax=1.3, volumeMin=1}}
 {id=Random_0, children [WavePlayer_2,WavePlayer_3,WavePlayer_4], properties {weights [1,1,1]}}
 {id=WavePlayer_2, properties {bLooping=true, soundWaveAssetPtr=&/Chrome_Anomaly_Pickup_Loop_03}}
 {id=WavePlayer_3, properties {bLooping=true, soundWaveAssetPtr=&/Chrome_Anomaly_Pickup_Loop_01}}
 {id=WavePlayer_4, properties {bLooping=true, soundWaveAssetPtr=&/Chrome_Anomaly_Pickup_Loop_02}}
]
properties {
 attenuationSettings=/Game/Athena/Sounds/SupplyDrops/Fort_HiddenStar_Ambient_Loop_Att
 maxDistance=2500
 soundClassObject=&Fort_Quests
 volumeMultiplier=1.2
 attenuationOverrides {falloffDistance=700, attenuationShapeExtents {x=100}}
}
