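# Sound cue for picking up a Chrome Anomaly in an Athena quest.
# A random node picks one of five pickup wave variations with equal weight.
# Overridden attenuation shapes distance falloff and stereo spread, with a boosted volume on the non-critical in-game UI sound class.
# Concurrency allows a single instance and blocks new plays while one is active.
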SoundCue /Game/Athena/Sounds/Quests/ChromeAnomalyPickup/Chrome_Anomaly_Pickup_Cue
root=Random_0
nodes [
 {id=Random_0, children [WavePlayer_1,WavePlayer_0,WavePlayer_2,WavePlayer_3,WavePlayer_4], properties {weights [1,1,1,1,1]}}
 {id=WavePlayer_0, properties {soundWaveAssetPtr=&/Chrome_Anomaly_PickUp_02}}
 {id=WavePlayer_1, properties {soundWaveAssetPtr=&/Chrome_Anomaly_PickUp_01}}
 {id=WavePlayer_2, properties {soundWaveAssetPtr=&/Chrome_Anomaly_PickUp_03}}
 {id=WavePlayer_3, properties {soundWaveAssetPtr=&/Chrome_Anomaly_PickUp_04}}
 {id=WavePlayer_4, properties {soundWaveAssetPtr=&/Chrome_Anomaly_PickUp_05}}
]
properties {
 bOverrideAttenuation=true
 bOverrideConcurrency=true
 maxDistance=1250
 soundClassObject=&Fort_UI_InGame_NonCritical
 volumeMultiplier=1.8
 attenuationOverrides {falloffDistance=850, stereoSpread=2300}
 concurrencyOverrides {maxCount=1, resolutionRule=PreventNew}
}
